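; Blueprint interface for objects locked behind a hero rank tag in the CherrySmoke hero ranking system, with chests implied by its function names.
; It declares a getter and setter for the required rank gameplay tag, a query returning two values of chest rank interaction details, and a call to show or hide a minimap icon.
; It holds only function signatures, so implementing assets supply all behavior.
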(fn GetChestRankInteractionDetails
  (return "" ""))

(fn GetHeroRankRequiredTag
  (return 0))

(fn SetHeroRankRequiredTag
  (params (UnlockTag :type Gameplay%20Tag%20Structure)))

(fn SetMinimapIconToShow
  (params (IsShowing :type Boolean)))
