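# EQS query that selects among hero-ranking discoverables, scored relative to the querier.
# A custom generator supplies the candidates, and a filter tests them against the Athena safe zone.
# A hero rank test, using the querier as player, filters and scores with the dominant weight, favoring lower values.
# Inverse-linear 3D distance scores then favor items near the querier and, at lower weight, near the current safe zone center.
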EnvQuery /CherrySmokeHeroRankingSystem/Gameplay/EQS_HeroRankingDiscoverables
options [{
  generator {
   name=FortQueryGenerator_HeroRankingDiscoverables_0
   class=FortQueryGenerator_HeroRankingDiscoverables
   properties {bAutoSortTests=true, bCanRunAsync=false, envQueryResultNormalizationOption=Default, optionName=""}
  }
  tests [
   FortQueryTest_InsideAthenaSafeZone_0 {class=FortQueryTest_InsideAthenaSafeZone, properties {testPurpose=Filter}}
   FortQueryTest_HeroRank_2 {
    class=FortQueryTest_HeroRank
    properties {
     bDefineReferenceValue=false
     boolValue=true
     clampMaxType=None
     clampMinType=None
     contextPlayer=&EnvQueryContext_Querier
     filterType=Minimum
     floatValueMax=0
     floatValueMin=0
     multipleContextFilterOp=AllPass
     multipleContextScoreOp=AverageScore
     normalizationType=Absolute
     referenceValue=0
     scoreClampMax=0
     scoreClampMin=0
     scoringEquation=InverseLinear
     scoringFactor=10
     testComment=""
     testPurpose=FilterAndScore
    }
   }
   EnvQueryTest_Distance_0 {
    class=EnvQueryTest_Distance
    properties {
     distanceTo=&EnvQueryContext_Querier
     filterType=Range
     floatValueMin=0
     scoringEquation=InverseLinear
     scoringFactor=1
     testComment=""
     testMode=Distance3D
     testPurpose=Score
    }
   }
   EnvQueryTest_Distance_1 {
    class=EnvQueryTest_Distance
    properties {
     distanceTo=/Script/FortniteAIServer.FortQueryContext_AthenaCurrentSafeZoneCenter
     filterType=Range
     floatValueMin=0
     scoringEquation=InverseLinear
     scoringFactor=0.5
     testComment=""
     testMode=Distance3D
     testPurpose=Score
    }
   }
  ]
 }]
